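# One-shot stereo MetaSound source for the FillerGrunt Viper pistol's FS sweetener, using weapon foley attenuation and sound class.
# On play, it shuffles through six sweetener waves and plays the selected one through a stereo wave player.
# Character affiliation, converted to a float, selects a gain for both channels, with one affiliation at full level and the others reduced.
# The player's finish trigger ends the one-shot.
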MetaSoundSource /FillerGrunt/Gameplay/Pistol_Fang/Audio/IP_MSS_FillerGrunt_Pistol_Viper_FS
inputs [
 Character.Affiliation {type=Enum:FortAudioAffiliation}
 Character.Health {type=Float}
 Character.UniqueID {type=Int32}
 FS Sweetener {type=WaveAsset:Array}
 UE.Source.OnPlay {type=Trigger}
]
interfaces [Fort.Character:1.0, UE.OutputFormat.Stereo:1.0, UE.Source.OneShot:1.0, UE.Source:1.1]
outputs [
 "UE.OutputFormat.Stereo.Audio:0" {type=Audio}
 "UE.OutputFormat.Stereo.Audio:1" {type=Audio}
 UE.Source.OneShot.OnFinished {type=Trigger}
]
pages [Default {
  input_defaults {
   Character.Affiliation=0
   Character.Health=0
   Character.UniqueID=0
   UE.Source.OnPlay=null
   FS Sweetener [
    /FillerGrunt/Gameplay/Pistol_Fang/Audio/Wavs/IP_FillerGrunt_ViperPistol_FS_Sweet_01
    /FillerGrunt/Gameplay/Pistol_Fang/Audio/Wavs/IP_FillerGrunt_ViperPistol_FS_Sweet_02
    /FillerGrunt/Gameplay/Pistol_Fang/Audio/Wavs/IP_FillerGrunt_ViperPistol_FS_Sweet_03
    /FillerGrunt/Gameplay/Pistol_Fang/Audio/Wavs/IP_FillerGrunt_ViperPistol_FS_Sweet_04
    /FillerGrunt/Gameplay/Pistol_Fang/Audio/Wavs/IP_FillerGrunt_ViperPistol_FS_Sweet_05
    /FillerGrunt/Gameplay/Pistol_Fang/Audio/Wavs/IP_FillerGrunt_ViperPistol_FS_Sweet_06
   ]
  }
  nodes [
   CrossfadeTriggerRouteFloat4_0 = "Crossfade.Trigger Route (Float, 4):Float"(Crossfade Value=Convert.Float:Int32(Int32=Convert.Int32:Enum:FortAudioAffiliation("Enum:FortAudioAffiliation"=^Character.Affiliation).Int32).Float) {also {defaults {In 0=0.65, In 1=0.65, In 2=1, In 3=0.65}}}
   n1 = Array.Shuffle:WaveAsset:Array(In Array=^FS Sweetener, Next=^UE.Source.OnPlay) {also {defaults {Enable Shared State="<MetaSoundBuilderResult.SUCCEEDED: 0>"}}}
   WavePlayer_0 = UE.Wave Player:Stereo(Play=@n1.On Next, Wave Asset=@n1.Value) {also {version [1,1]}}
   Multiply_0 = UE.Multiply:Audio by Float(AdditionalOperands=@CrossfadeTriggerRouteFloat4_0.Out, PrimaryOperand=@WavePlayer_0.Out Left)
   Multiply_1 = UE.Multiply:Audio by Float(AdditionalOperands=@CrossfadeTriggerRouteFloat4_0.Out, PrimaryOperand=@WavePlayer_0.Out Right)
  ]
  links [
   {dst="@UE.OutputFormat.Stereo.Audio:0", src=Multiply_0.Out}
   {dst="@UE.OutputFormat.Stereo.Audio:1", src=Multiply_1.Out}
   {dst="@UE.Source.OneShot.OnFinished", src=WavePlayer_0.On Finished}
  ]
 }]
properties {
 assetImportData=&self:AssetImportData
 attenuationSettings=&AS_Fort_WeaponFoley
 maxDistance=2250
 numChannels=2
 outputFormat=Stereo
 sampleRate=48000
 soundClassObject=&Fort_WeaponFoley
 volume=0.3
}
