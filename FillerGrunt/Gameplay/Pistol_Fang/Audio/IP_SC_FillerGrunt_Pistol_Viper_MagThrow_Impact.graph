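# Sound cue for the impact of a thrown magazine from the FillerGrunt Viper (Fang) pistol.
# A Random node picks among three impact wave variations with equal weight and plays it at reduced volume.
# It uses close-range mono melee attack attenuation and player melee weapon concurrency, but routes to the medium explosives sound class.
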SoundCue /FillerGrunt/Gameplay/Pistol_Fang/Audio/IP_SC_FillerGrunt_Pistol_Viper_MagThrow_Impact
root=Random_0
nodes [
 {id=Random_0, children [WavePlayer_3,WavePlayer_4,WavePlayer_5], properties {weights [1,1,1]}}
 {id=WavePlayer_3, properties {soundWaveAssetPtr=/FillerGrunt/Gameplay/Pistol_Fang/Audio/Wavs/IP_FillerGrunt_ViperPistol_MagThrow_Impact_01}}
 {id=WavePlayer_4, properties {soundWaveAssetPtr=/FillerGrunt/Gameplay/Pistol_Fang/Audio/Wavs/IP_FillerGrunt_ViperPistol_MagThrow_Impact_02}}
 {id=WavePlayer_5, properties {soundWaveAssetPtr=/FillerGrunt/Gameplay/Pistol_Fang/Audio/Wavs/IP_FillerGrunt_ViperPistol_MagThrow_Impact_03}}
]
properties {
 attenuationSettings=/Game/Sounds/Attenuation/Fort_Weapon_Melee_Attack_Close_Range_Mono_Att
 maxDistance=10058.75
 soundClassObject=/Game/Sounds/FortSoundClassesAndModes/Fort_Explosives_Medium
 volumeMultiplier=0.6
 concurrencySet [&Player_Melee_Weapon_Conc]
}
