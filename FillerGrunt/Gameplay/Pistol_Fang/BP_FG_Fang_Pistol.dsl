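; Weapon Blueprint for the Filler Grunt Fang pistol, holding leftover prototype hooks for a throw-magazine reload.
; ThrowMag declares reload, owner, spawn transform and enemy parameters but has no body.
; SetMagThrowHUD and CacheMagThrowValidInRange both just store the MagThrow_ValidInRange flag, and are marked for deletion once the real mag projectile works.
; BeginPlay and construction only call the parent, and Tick is empty.
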(event EventBeginPlay
  (|CallParentFunction))

(event EventTick
  (params (DeltaSeconds :type Float%20%28single-precision%29)))

(event Custom|ThrowMag
  (params (ReloadTime :type Float%20%28single-precision%29) (ReloadType :type EFortWeaponReloadType%20Enum :default "ReloadWholeClip") (Owner :type Actor%20Object%20Reference) (SpawnLoc :type Vector) (SpawnRot :type Rotator) (Enemy :type Actor%20Object%20Reference)))

(event Custom|SetMagThrowHUD
  (params (ValidInRange :type Boolean))
  ; Old Proto Throw Mag - DELETE THIS WHEN REAL MAG PROJECTILE IS WORKING
  (Variables|Default|SetMagThrow_ValidInRange ValidInRange))

(event Custom|CacheMagThrowValidInRange
  (params (ValidInRange :type Boolean))
  ; Old Proto Throw Mag - DELETE THIS WHEN REAL MAG PROJECTILE IS WORKING
  (Variables|Default|SetMagThrow_ValidInRange ValidInRange))

(fn ConstructionScript
  (|CallParentFunction))
