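; BP_MagToss is an actor for the Fang pistol that, going by its name, is a tossed magazine flung as a physics object.
; On BeginPlay, only the network authority enables simulation on the static mesh and applies a mass-independent impulse along its forward vector with a slight upward component.
; The authority destroys the actor after a short delay.
; Sphere overlap, actor overlap and Tick events are empty, so it has no hit logic.
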(event EventBeginPlay
  (bind _staticmesh (Variables|Default|GetStaticMesh))
  (Networking|SwitchHasAuthority
    (:Authority
      (Physics|SetSimulatePhysics _staticmesh true)
      (Physics|AddImpulse _staticmesh (+ (* (Transformation|GetUpVector _staticmesh) 5e+02) (* (Transformation|GetForwardVector _staticmesh) 4e+03)) "None" true)
      (Utilities|FlowControl|Delay 3.0)
      (Actor|DestroyActor))))

(event Collision|EventActorBeginOverlap
  (params (OtherActor :type Actor%20Object%20Reference)))

(event EventTick
  (params (DeltaSeconds :type Float%20%28single-precision%29)))

(event OnComponentBeginOverlap(Sphere)
  (params (OverlappedComponent :type Primitive%20Component%20Object%20Reference) (OtherActor :type Actor%20Object%20Reference) (OtherComp :type Primitive%20Component%20Object%20Reference) (OtherBodyIndex :type Integer) (bFromSweep :type Boolean) (SweepResult :type Hit%20Result%20Structure%20%28by%20ref%29)))

(fn ConstructionScript)
